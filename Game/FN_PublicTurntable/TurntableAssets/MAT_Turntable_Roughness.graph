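# Post-process material in the turntable assets that visualizes scene roughness, judging by its name and the buffer it reads.
# A shading path switch passes the Roughness scene texture color on the default path.
# On the mobile path the switch substitutes a constant of one instead of sampling that buffer.
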Material /Game/FN_PublicTurntable/TurntableAssets/MAT_Turntable_Roughness
graph {
 properties {materialDomain=MD_PostProcess, pixelDepthOffsetMode=PDOM_Legacy, usedShadingModels=Default Lit}
 nodes [ShadingPathSwitch(Default=SceneTexture() {sceneTextureId=PPI_Roughness}.Color, Mobile=Constant() {r=1})]
}
