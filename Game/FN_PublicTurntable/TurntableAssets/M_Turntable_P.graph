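# M_Turntable_P is a parameterized Default Lit material, named for the turntable in the FN_PublicTurntable assets.
# It defines a Base Color vector parameter with a dark grey default and a Metallic scalar parameter, so material instances can adjust both.
# The graph holds no textures or math nodes, and shows no links from the two parameters to material outputs.
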Material /Game/FN_PublicTurntable/TurntableAssets/M_Turntable_P
graph {
 properties {pixelDepthOffsetMode=PDOM_Legacy, usedShadingModels=Default Lit}
 nodes [
  ScalarParameter() {parameterName=Metallic}
  VectorParameter() {parameterName=Base Color, defaultValue=<rgba 0.08 0.08 0.08 0>}
 ]
}
